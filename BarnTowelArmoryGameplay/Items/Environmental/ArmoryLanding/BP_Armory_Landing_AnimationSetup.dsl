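; Blueprint that drives the armory landing animation setup for a mobile armory environmental item.
; A custom AnimateArmoryLanding event writes the current game time into the BP_GetTime scalar parameter on the PodCatcher_03 component's materials.
; The materials thus receive a start timestamp, likely for shader-driven animation.
; The entrance ramp's component-activated event and the construction script are empty.
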(event Custom|AnimateArmoryLanding
  (Rendering|Material|SetScalarParameterValueonMaterials (Variables|Default|GetMobileArmory_PodCatcher_03) "BP_GetTime" (Utilities|Time|GetGameTimeinSeconds)))

(event OnComponentActivated(MobileArmory_EntranceRamp)
  (params (Component :type Actor%20Component%20Object%20Reference) (bReset :type Boolean)))

(fn ConstructionScript)
