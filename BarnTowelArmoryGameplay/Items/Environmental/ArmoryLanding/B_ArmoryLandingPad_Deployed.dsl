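; Deployed Armory landing pad that blocks building per a hotfix value, animates its catcher arms and dampener, and plays deploy sounds.
; Replicated begin/played flags let clients present at landing see the animation, while later arrivals get the finished pose.
; Staggered material time parameters drive the animation.
; A hotfix-timed completion after landing advances the spawn manager's Armory state to door opening, removes the thruster wind cue, and enables collision disabled during deployment.
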(event EventBeginPlay
  ;[0 1 0.01736] BeginPlay
  ; Start the deploying animation and sounds for clients in relevancy, set the final state for players entering relevancy later
  ;;[0.4987 0.7268 1] Setup the Build rules based on the hotfix row
  ;; - this asset only needs to deal with the "On Landed" row since it activated when the Armory is landed
  (Variables|Building|SetbForceBlockBuildings (ScalableFloat|GetValueasBool (Variables|Default|GetHFRow_DisablePlayerBuildingOnPadLanded))) ; blocks player building for overlapping structures
  (CallFunction|SetCollisiononAnimatedArmsandRamp (Variables|Default|GetbHasLandingAnimationPlayed)) ; disable collision on the animated elements of the \nlanding pad since they are not yet "deployed" and \nmay interfere with the player knockback attempting \nto clear them from the area (FORT-784284)
  ;;[0.4987 0.7268 1] Begin and replicate the Landing Pod animation if the animation has not yet finished
  (Networking|SwitchHasAuthority
    (:Authority
      (Variables|Default|SetbShouldBeginLandingAnimation true))))

(event OnArmoryLanded(BP_BarnTowelArmory_ReactiveProp_Component)
  (params (SelectedVariant :type Enum_BarnTowelArmory_SelectedVariant%20Enum :default "NewEnumerator0"))
  (bind _hfrow_landingpadanimationduration (Variables|Default|GetHFRow_LandingPadAnimationDuration))
  (bind _returnvalue (ScalableFloat|GetValueatLevel _hfrow_landingpadanimationduration))
  ;[0.02214 0.342 0.1684] OnArmoryLanded
  ; Play any landing pad animations that should be triggered by landing
  (if (> _returnvalue 0)
    (Utilities|Time|SetTimerbyEvent (EventDispatchers|CreateEvent :!event Custom|LandingPadAnimationCompleted) _returnvalue) ; use timer to determine when the Armory animation is completed\nMaterial animations do not have events to key off of, so the timer\nis set to the matching expected value
    (else
      (CallFunction|LandingPadAnimationCompleted))))

(event OnArmoryThrustersActivated(BP_BarnTowelArmory_ReactiveProp_Component)
  ;[0.3563 0 0.8073] Play landing ground wind GC when thrusters activate
  (GameplayCue|AddGameplayCue self (Variables|Default|GetGC_GroundWind)))

(event Custom|LandingPadAnimationCompleted
  ;[0.02214 0.342 0.1684] OnLandingPadAnimationCompleted
  ; Switch the Armory Manager state to "Opening"
  (bind _owningarmoryspawnmanager (Variables|/BarnTowelArmoryGameplay/Items/BP_BarnTowelArmory_ReactiveProp_Component.BP_BarnTowelArmory_ReactiveProp_Component_C|GetOwningArmorySpawnManager (Variables|Default|GetBP_BarnTowelArmory_ReactiveProp_Component))
    (:then
      (Class|BBarnTowelSpawnManager|SetNewArmoryState _owningarmoryspawnmanager "NewEnumerator5") ; once the landing animations are complete, \ntrigger the Armory door opening animations
      (GameplayCue|RemoveGameplayCue self (Variables|Default|GetGC_GroundWind))
      (Variables|Default|SetbHasLandingAnimationPlayed true) ; new clients entering relevancy should not\nplay the deploying animation
    )
    (:else)))

(event Custom|InitializePadAnimationByRelevancy
  ;[0.3724 0.6306 1 0.5] InitializePadAnimationByRelevancy
  ; The pod animation should only be played for clients in relevancy when the armory lands, otherwise it should default to the deployed state
  (Networking|CanExecuteCosmeticEvents
    (:True
      (if (Variables|Default|GetbShouldBeginLandingAnimation)
        (if (Variables|Default|GetbHasLandingAnimationPlayed)
          (Utilities|FlowControl|DoOnce false
            (:Completed
              (CallFunction|BeginPadAnimation 0.0) ; setting the animation time to zero defaults \nthe visuals to the ending state
            )) ; these bools are replicated individually in the \noff chance they would get out of sync. \nbut if they both replicate on the same tick, \nwe only need it to run one time
          (else
            (Utilities|FlowControl|DoOnce false
              (:Completed
                (CallFunction|BeginPadAnimation (Utilities|Time|GetGameTimeinSeconds))
                (CallFunction|PlayPadDeployingSounds)))))))))

(event Custom|BeginPadAnimation
  (params (AnimationTime :type Float%20%28double-precision%29))
  ;[0.3563 0 0.8073] Pad Animation (animation parameters are on the materials themselves)
  (Rendering|Material|SetScalarParameterValueonMaterials (Variables|Default|GetMobileArmory_PodDampener) "BP_GetTime" AnimationTime)
  (Utilities|FlowControl|Delay (Variables|Default|GetPodCatcherOverlappingDelay))
  (Rendering|Material|SetScalarParameterValueonMaterials (Variables|Default|GetMobileArmory_PodCatcher_03) "BP_GetTime" AnimationTime)
  (Utilities|FlowControl|Delay (Variables|Default|GetPodCatcherOverlappingDelay))
  (Rendering|Material|SetScalarParameterValueonMaterials (Variables|Default|GetMobileArmory_PodCatcher_02) "BP_GetTime" AnimationTime)
  (Utilities|FlowControl|Delay (Variables|Default|GetPodCatcherOverlappingDelay))
  (Rendering|Material|SetScalarParameterValueonMaterials (Variables|Default|GetMobileArmory_PodCatcher_01) "BP_GetTime" AnimationTime)
  (Rendering|Components|Light|SetIntensity (Variables|Default|GetSpotlight%203) (Math|Float|Lerp 0.0 (Variables|Default|GetLightIntensity) AnimationTime)))

(event Custom|PlayPadDeployingSounds
  ;[0.3563 0 0.8073] Pad Animation (animation parameters are on the materials themselves)
  ;;[0.15 0.08091 0.09554 0.5] Stairs Sound - material animations do not have events to key off of for sounds, so these can be activated independently
  (Networking|SwitchHasAuthority
    (:Remote
      (Audio|SpawnSoundAttached (Variables|Sound|GetStairsRiseSound) (Variables|Default|GetMobileArmory_EntranceRamp))
      (Audio|SpawnSoundAttached (Variables|Sound|GetPodCatcherArmsUpSound) (Variables|Default|GetMobileArmory_PodCatcher_04)))))

(fn OnRep_bHasLandingAnimationPlayed
  (CallFunction|InitializePadAnimationbyRelevancy)
  (CallFunction|SetCollisiononAnimatedArmsandRamp (Variables|Default|GetbHasLandingAnimationPlayed)) ; enable the collision on the "deployed" elements now \nthat they are fully visible and the knockback effect is ending
)

(fn OnRep_bShouldBeginLandingAnimation
  (CallFunction|InitializePadAnimationbyRelevancy))

(fn SetCollisionOnAnimatedArmsAndRamp
  (params (SetCollisionEnabled :type Boolean))
  (Collision|SetCollisionEnabled (Variables|Default|GetMobileArmory_PodCatcher_03) (select SetCollisionEnabled "QueryAndPhysics" "NoCollision")))

(fn ConstructionScript)
