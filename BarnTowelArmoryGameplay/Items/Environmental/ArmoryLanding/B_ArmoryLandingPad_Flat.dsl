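; Flat armory landing pad Blueprint that sets player building restrictions when it begins play; it gets replaced on landing.
; BeginPlay runs the parent logic, then reads a hotfixable scalable float row for disabling player building on start.
; That boolean sets force-block on overlapping building and toggles the NoBuildZone collision between query-only and none.
; The no-build zone handles deployables such as Port-A-Bunkers, and the construction script is empty.
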(event EventBeginPlay
  (bind _hfrow_disableplayerbuildingonstart (Variables|Default|GetHFRow_DisablePlayerBuildingOnStart))
  (bind _returnvalue (ScalableFloat|GetValueasBool _hfrow_disableplayerbuildingonstart))
  (|CallParentFunction)
  ;[0.4987 0.7268 1] Setup the Build rules based on the hotfix row
  ; - this asset only needs to deal with the "On Start" row since it is replaced on Landing
  (Variables|Building|SetbForceBlockBuildings _returnvalue) ; blocks player building for overlapping structures
  (Collision|SetCollisionEnabled (Variables|Default|GetNoBuildZone) (select _returnvalue "QueryOnly" "NoCollision")) ; "No Build Zones" handle deployables, \n(such as Port-A-Bunkers)
)

(fn ConstructionScript)
