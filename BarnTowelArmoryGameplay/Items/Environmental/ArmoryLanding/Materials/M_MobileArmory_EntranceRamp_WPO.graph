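# Lit material for the Mobile Armory entrance ramp that animates the ramp into its modeled pose through world position offset.
# Time minus a delay and a BP_GetTime start value, scaled by speed, drives a smoothstep; the huge default start time holds the initial pose.
# That curve eases out a pivot rotation around local X plus a positional offset.
# The normal map passes through the same rotation function and returns to tangent space.
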Material /BarnTowelArmoryGameplay/Items/Environmental/ArmoryLanding/Materials/M_MobileArmory_EntranceRamp_WPO
graph {
 properties {bAutomaticallySetUsageInEditor=false, usedShadingModels=Default Lit}
 nodes [
  n1 = Subtract(
   A=Subtract(A=Time(), B=ScalarParameter() {group=Ramp_Animation, parameterName=Ramp_Delay})
   B=ScalarParameter() {defaultValue=100000000, group=Ramp_Animation, parameterName=BP_GetTime}
  )
  n2 = ScalarParameter() {defaultValue=1, group=Ramp_Animation, parameterName=Ramp_Speed}
  n3 = SmoothStep(Value=Saturate(None=Multiply(A=@n1, B=@n2)))
  n4 = Constant3Vector() {constant=<rgba 1 0 0 0>}
  n5 = TextureSampleParameter2D() {
   group=Basic_PBR
   parameterName=Normals_Texture
   samplerType=SAMPLERTYPE_Normal
   texture=/BarnTowelArmoryGameplay/Items/Environmental/ArmoryLanding/Textures/T_MobileArmory_Launch_Pad_Ramp_N
  }
  n6 = MaterialFunctionCall!(
   RotationAngle=LinearInterpolate(A=Constant() {r=0.069}, Alpha=@n3) {constB=0}
   NormalizedRotationAxis=Transform(None=Normalize(VectorInput=@n4)) {transformSourceType=TRANSFORMSOURCE_Local}
   PivotPoint=TransformPosition(None=VectorParameter() {group=Ramp_WPO, parameterName=RampPivot, defaultValue=<rgba 0 0 0 1>}.RGB) {transformType=TRANSFORMPOSSOURCE_World}
   Position
   Normal=Transform(None=@n5.RGB)
   TangentU
   TangentV
  ) {materialFunction=/Game/Characters/Player/Male/Medium/Bodies/M_MED_Inferno/FX/Mat/RotateVertexAboutAxis}
  n7 = VectorParameter() {group=Ramp_WPO, parameterName=Ramp_PositionOffset, defaultValue=<rgba 0 0 1 1>}
  n8 = OneMinus(None=@n3)
  Add(A=@n6.WPO, B=Transform(None=Multiply(A=@n7.RGB, B=@n8)) {transformSourceType=TRANSFORMSOURCE_Local})
  TextureSampleParameter2D() {
   group=Basic_PBR
   parameterName=BaseColor_Texture
   texture=/BarnTowelArmoryGameplay/Items/Environmental/ArmoryLanding/Textures/T_MobileArmory_Launch_Pad_Ramp_D
  }
  TextureSampleParameter2D() {
   group=Basic_PBR
   parameterName=SRM_Texture
   samplerType=SAMPLERTYPE_Masks
   texture=/BarnTowelArmoryGameplay/Items/Environmental/ArmoryLanding/Textures/T_MobileArmory_Launch_Pad_Ramp_S
  }
  Transform(None=@n6.Normal) {transformSourceType=TRANSFORMSOURCE_World, transformType=TRANSFORM_Tangent}
 ]
}
