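# Mobile Armory pod catcher material that animates lids, top and bottom catchers and a rise-up through World Position Offset on a GigaMix 2023 layered base.
# Vertex color red and pre-skinning local position build per-part masks and lid directions.
# BP_GetTime, with per-part delays and speeds, drives smoothstepped offsets along an angled axis and vertically.
# Shared effect, seasonal, RVT and low-quality mobile stages precede the final WPO override.
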Material /BarnTowelArmoryGameplay/Items/Environmental/ArmoryLanding/Materials/M_MobileArmory_PodCatcher_Lids_WPO_PBR
graph {
 properties {
  bAutomaticallySetUsageInEditor=false
  bTangentSpaceNormal=false
  bUseMaterialAttributes=true
  bUsedWithInstancedStaticMeshes=true
  bUsedWithNanite=true
  bUsedWithSplineMeshes=true
  ditheredLODTransition=true
  pixelDepthOffsetMode=PDOM_Legacy
  shadingModel=MSM_FromMaterialExpression
  usedShadingModels=Default Lit | Subsurface | Clear Coat | Two Sided Foliage
 }
 nodes [
  n1 = MaterialFunctionCall() {materialFunction=/Game/Global/Base4Layers_2022/MaterialFunctions/MF_GigaMix_2023}
  MaterialExpressionReroute_70 = Reroute(None=@n1.Result)
  n2 = ScalarParameter() {
   desc=Documentation link
   group="0 (Expand)  Documentation (RMB - Copy display name) and Debug view"
   parameterName=https://confluence.it.epicgames.com/display/FORT/Fortnite+Base+Material+2022
  }
  n3 = SetMaterialAttributes!(MaterialAttributes=@MaterialExpressionReroute_70.MaterialAttributes, World Position Offset=Multiply(A=@n2)) {attributeSetTypes [#1]}
  MaterialExpressionReroute_71 = Reroute(None=@n3)
  MaterialExpressionReroute_69 = Reroute(None=@MaterialExpressionReroute_71.Output)
  n4 = MaterialFunctionCall!(Material Attributes=MaterialFunctionCall!(In=MaterialFunctionCall(Material_In=@MaterialExpressionReroute_69.MaterialAttributes) {materialFunction=&MF_FN_DestructibleDamage}.Material_Out) {materialFunction=/Game/Athena/WinterArt/Materials/MF_WPO_WindNew}.Result) {materialFunction=/Game/Creative/Sets/MilitaryBase/Props/MF_FanSpin}
  n5 = MaterialFunctionCall!(In=MaterialFunctionCall!(
    In=MaterialFunctionCall!(In=@n4.Result, UseSphericalMappingOverride) {materialFunction=/Game/Packages/Fortress_SharedMaterials/Base_Material_Functions/MF_CurieFallbackEffect}.Result
    Custom Mask
    Custom Rotation Axis
   ) {materialFunction=/Game/Athena/WinterArt/Materials/MF_WindWPO_v2}.Result) {materialFunction=/Game/Global/Base4Layers_2022/MaterialFunctions/MF_SnowAndIceArchitecture}
  MaterialExpressionReroute_68 = Reroute(None=@n5.Result)
  Composite!(None=@MaterialExpressionReroute_68.Result) {desc=Transform to World in case we will use RVT after this point, subgraphName=Tangent to World}
  Composite!(MaterialAttributes=@n1.Result) {subgraphName=Documentation Link}
  Composite(MaterialAttributes=@MaterialExpressionReroute_71.Output) {desc=Game Shared WPO and Fire Effect, subgraphName=WPO And Shared Functions}
  MaterialExpressionReroute_1 = Reroute(None=@n12.RVT H)
  MaterialExpressionReroute_3 = Reroute(None=@MaterialExpressionReroute_68.Result)
  n6 = GetMaterialAttributes(None=@MaterialExpressionReroute_3.Input) {attributeGetTypes [#2]}
  n7 = Transform(None=@n6.Normal)
  n8 = LinearInterpolate(A=Multiply(A=@n7) {constB=-1}, B=@n7, Alpha=Saturate(None=TwoSidedSign())) {desc="Fixes normals\r\non backfaces"}
  n9 = SetMaterialAttributes!(
   MaterialAttributes=@n6.MaterialAttributes
   Normal=StaticSwitchParameter(True=MaterialProxyReplace(Realtime=@n8, MaterialProxy=@n6.Normal), False=@n8) {parameterName=Force TangentSpaceNormalsFor HLOD ProxyReplace}
  ) {attributeSetTypes [#2]}
  MaterialExpressionReroute_2 = Reroute(None=@n9)
  n10 = MaterialFunctionCall!(In=MaterialFunctionCall!(In=@MaterialExpressionReroute_2.Output) {materialFunction=/Game/ArtTools/Tech_Art/DirectFX/Assets/MaterialFunctions/MF_DirectFX}.Result) {materialFunction=/Game/ArtTools/Tech_Art/Veneer/MaterialFunctions/MF_Veneer}
  MaterialExpressionReroute_8 = Reroute(None=@n10.Result)
  n11 = MaterialFunctionCall!(AO, RVT_Height=@MaterialExpressionReroute_1."RVT_Height (S)", For Buildings=@MaterialExpressionReroute_8."For Moss (MA)", For Moss) {materialFunction=/Game/Environments/Asteria/Rocks/Materials/MaterialFunctions/MF_JungleBiome}
  MaterialExpressionReroute_0 = Reroute(None=@n11.Buildings Jungl)
  n12 = MaterialFunctionCall!(
   In=@MaterialExpressionReroute_0.Result
   "RVT Height Map (Roughness)"=GetMaterialAttributes(None=@n10.Result) {attributeGetTypes [#3]}.Roughness
  ) {materialFunction=/Game/Global/Base4Layers_2022/MaterialFunctions/MF_RVTBlending}
  Composite!(InputPin=@n12.Result) {subgraphName=Mobile Opt}
  Composite!("For Moss (MA)"=@n10.Result, "RVT_Height (S)"=@n12.RVT H) {subgraphName="Seasonal "}
  n13 = VertexColor()
  n14 = LocalPosition() {includedOffsets=ExcludeOffsets, localOrigin=InstancePreSkinning}
  n15 = ShaderStageSwitch(PixelShader=VertexInterpolator(VS=@n14.XYZ).PS, VertexShader=@n14.XYZ)
  n16 = Saturate(None=ComponentMask(None=@n15) {b=true})
  n17 = NamedRerouteDeclaration(None=Multiply(A=@n13.R, B=OneMinus(None=@n16))) {name=Lids_Mask, nodeColor=<rgba 0.9377 1 0.8953 1>}
  n18 = ComponentMask(None=@n15) {r=true}
  n19 = OneMinus(None=Saturate(None=Add(A=@n18) {constB=-350}))
  n20 = Saturate(None=Add(A=ComponentMask(None=@n15) {g=true}) {constB=-400})
  n21 = Multiply(A=@n17, B=Saturate(None=Add(A=@n19, B=@n20)))
  NamedRerouteDeclaration(None=Add(A=@n17, B=Multiply(A=@n21) {constB=-2})) {name=Lids_Orientation, nodeColor=<rgba 0 1 0.1328 1>}
  n22 = Multiply(A=Subtract(A=@n13.R) {constB=0.6}) {constB=10000}
  NamedRerouteDeclaration(None=Multiply(A=Saturate(None=@n22), B=@n16)) {name=TopCatcher_Mask, nodeColor=<rgba 1 0.0469 0 1>}
  n23 = Multiply(A=@n13.R, B=OneMinus(None=Multiply(A=@n13.R) {constB=1.5}))
  NamedRerouteDeclaration(None=Saturate(None=Multiply(A=@n23) {constB=100000})) {name=BottomCatcher_Mask, nodeColor=<rgba 0 0.2734 1 1>}
  n24 = Time()
  n25 = Add(
   A=ScalarParameter() {group=PodCatcher_Animation, parameterName=RisingUp_Delay, sortPriority=11}
   B=ScalarParameter() {group=PodCatcher_Animation, parameterName=Overall_Delay, sortPriority=11}
  )
  n26 = Multiply(
   A=ScalarParameter() {defaultValue=1, group=PodCatcher_Animation, parameterName=RisingUp_Speed, sortPriority=20}
   B=Subtract(
    A=Subtract(A=@n24, B=@n25) {constB=0}
    B=ScalarParameter() {defaultValue=9999999, group=PodCatcher_Animation, parameterName=BP_GetTime, sortPriority=0}
   )
  )
  NamedRerouteDeclaration(None=SmoothStep(Value=Saturate(None=@n26))) {name=RisingUp_AnimationTime, nodeColor=<rgba 1 0.2812 0 1>}
  n27 = Multiply(A=NamedRerouteUsage(), B=ScalarParameter() {defaultValue=1, group=PodCatcher_Animation, parameterName=Lids_Speed, sortPriority=20})
  n28 = Multiply(A=NamedRerouteUsage(), B=ScalarParameter() {defaultValue=1, group=PodCatcher_Animation, parameterName=TopCatcher_Speed, sortPriority=11})
  n29 = NamedRerouteUsage()
  n30 = ScalarParameter() {defaultValue=1, group=PodCatcher_Animation, parameterName=BottomCatcher_Speed, sortPriority=11}
  n31 = Time()
  n32 = Multiply(A=NamedRerouteUsage(), B=ScalarParameter() {group=PodCatcher_Animation, parameterName=Lids_Delay, sortPriority=20})
  n33 = Multiply(A=NamedRerouteUsage(), B=ScalarParameter() {group=PodCatcher_Animation, parameterName=TopCatcher_Delay, sortPriority=11})
  n34 = NamedRerouteUsage()
  n35 = ScalarParameter() {group=PodCatcher_Animation, parameterName=BottomCatcher_Delay, sortPriority=11}
  n36 = Add(
   A=Add(A=Add(A=@n32, B=@n33), B=Multiply(A=@n34, B=@n35))
   B=ScalarParameter() {group=PodCatcher_Animation, parameterName=Overall_Delay, sortPriority=11}
  )
  n37 = Multiply(
   A=Add(A=Add(A=@n27, B=@n28), B=Multiply(A=@n29, B=@n30))
   B=Subtract(
    A=Subtract(A=@n31, B=@n36) {constB=0}
    B=ScalarParameter() {defaultValue=9999999, group=PodCatcher_Animation, parameterName=BP_GetTime, sortPriority=0}
   )
  ) {constA=1}
  NamedRerouteDeclaration(None=SmoothStep(Value=Saturate(None=@n37))) {name=AnimationTime, nodeColor=<rgba 0.3594 0 1 1>}
  n38 = ScalarParameter() {defaultValue=30, group=PodCatcher_WPO, parameterName=RotationAngle}
  n39 = Cosine(None=@n38) {period=360}
  n40 = Sine(None=@n38) {period=360}
  n41 = Constant()
  NamedRerouteDeclaration(None=Normalize(VectorInput=MaterialFunctionCall(R=@n39, G=@n40, B=@n41) {materialFunction=&AppendMany}.RGB)) {name=RotationAxisVector, nodeColor=<rgba 0 1 0.0156 1>}
  n42 = ScalarParameter() {defaultValue=-166, group=PodCatcher_WPO, parameterName=Lids_ClosePosition}
  n43 = NamedRerouteUsage()
  n44 = Multiply(A=NamedRerouteUsage(), B=Multiply(A=NamedRerouteUsage(), B=LinearInterpolate(A=@n42, Alpha=@n43) {constA=-166, constB=0}))
  n45 = ScalarParameter() {defaultValue=17, group=PodCatcher_WPO, parameterName=Lids_ClosedHeightOffset}
  n46 = SmoothStep(Value=@n43) {constMax=0.6}
  n47 = NamedRerouteUsage()
  n48 = Constant3Vector() {constant=<rgba 0 0 1 0>}
  n49 = LinearInterpolate(Alpha=NamedRerouteUsage()) {constA=-440, constB=0}
  n50 = OneMinus(None=NamedRerouteUsage())
  n51 = Add(A=Multiply(A=LinearInterpolate(A=@n45, Alpha=@n46) {constB=0}, B=Multiply(A=@n47, B=@n48)), B=Multiply(A=@n48, B=Multiply(A=@n49, B=@n50)))
  NamedRerouteDeclaration(None=Transform(None=Add(A=@n44, B=@n51)) {transformSourceType=TRANSFORMSOURCE_Local}) {name=OutputWPO, nodeColor=<rgba 1 0.6563 0 1>}
  PinBase() {reroutePins [Input {expression=&self:MaterialExpressionReroute_3}]}
  PinBase!(Output=@n9) {reroutePins [Output {expression=&self:MaterialExpressionReroute_2}]}
  PinBase!() {reroutePins [
    "For Moss (MA)" {expression=&self:MaterialExpressionReroute_8}
    "RVT_Height (S)" {expression=&self:MaterialExpressionReroute_1}
   ]}
  PinBase!(Result=@n11.Buildings Jungl) {reroutePins [Result {expression=&self:MaterialExpressionReroute_0}]}
  PinBase!() {reroutePins [MaterialAttributes {expression=&self:MaterialExpressionReroute_70}]}
  PinBase!(Output=@n3) {reroutePins [Output {expression=&self:MaterialExpressionReroute_71}]}
  PinBase!() {reroutePins [MaterialAttributes {expression=&self:MaterialExpressionReroute_69}]}
  PinBase!(Result=@n5.Result) {reroutePins [Result {expression=&self:MaterialExpressionReroute_68}]}
  PinBase!() {reroutePins [InputPin {expression=&self:MaterialExpressionReroute_7}]}
  MaterialExpressionReroute_7 = Reroute(None=@n12.Result)
  n52 = GetMaterialAttributes(None=@MaterialExpressionReroute_7.InputPin)
  n53 = Constant() {r=0.5}
  n54 = Constant() {r=0.8}
  n55 = ShadingPathSwitch(
   Default=@MaterialExpressionReroute_7.InputPin
   Mobile=QualitySwitch(
    Default=@MaterialExpressionReroute_7.InputPin
    Low=SetMaterialAttributes!(MaterialAttributes=@n52.MaterialAttributes, Specular=@n53, Roughness=@n54) {attributeSetTypes [#4, #3]}
    High=@MaterialExpressionReroute_7.InputPin
    Medium=@MaterialExpressionReroute_7.InputPin
    Epic=@MaterialExpressionReroute_7.InputPin
   )
  )
  PinBase!(Output=@n55) {reroutePins [Output {expression=&self:MaterialExpressionReroute_6}]}
  MaterialExpressionReroute_6 = Reroute(None=@n55)
  SetMaterialAttributes!(
   MaterialAttributes=MaterialFunctionCall(MA In=@MaterialExpressionReroute_6.Output) {materialFunction=&MF_CreativeMoveTool}.MA Out
   World Position Offset=NamedRerouteUsage()
  ) {attributeSetTypes [#1]}
 ]
}
