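# Nanite-enabled lit material for the Mobile Armory pod catcher, animating its lids, top and bottom catchers and rise through world position offset.
# Vertex red and local position build per-part masks and lid directions.
# Time minus BP_GetTime, with delay and speed parameters, drives smoothstepped animation curves.
# These curves blend pivot rotation of the catchers with local-space lid and height offsets; the normal map is rotated along with the vertices.
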Material /BarnTowelArmoryGameplay/Items/Environmental/ArmoryLanding/Materials/M_MobileArmory_PodCatcher_WPO
graph {
 properties {bAutomaticallySetUsageInEditor=false, bUsedWithNanite=true, usedShadingModels=Default Lit}
 nodes [
  n1 = NamedRerouteUsage()
  n2 = LinearInterpolate(A=ScalarParameter() {group=PodCatcher_WPO, parameterName=TopCatcher_MaxRotation, sortPriority=40}, Alpha=@n1) {constB=0}
  n3 = NamedRerouteUsage()
  n4 = LinearInterpolate(A=ScalarParameter() {group=PodCatcher_WPO, parameterName=BottomCatcher_MaxRotation, sortPriority=40}, Alpha=@n1) {constB=0}
  n5 = NamedRerouteUsage()
  n6 = TextureSampleParameter2D() {
   group=Basic_PBR
   parameterName=Normals_Texture
   samplerSource=SSM_Wrap_WorldGroupSettings
   samplerType=SAMPLERTYPE_Normal
   texture=/BarnTowelArmoryGameplay/Items/Environmental/ArmoryLanding/Textures/T_MobileArmory_Launch_Pad_Ramp_N
  }
  n7 = MaterialFunctionCall!(
   RotationAngle=Add(A=Multiply(A=@n2, B=@n3), B=Multiply(A=@n4, B=@n5))
   NormalizedRotationAxis=Transform(None=NamedRerouteUsage()) {transformSourceType=TRANSFORMSOURCE_Local}
   PivotPoint=TransformPosition(None=VectorParameter() {group=PodCatcher_WPO, parameterName=RotationPivot, defaultValue=<rgba 5.2 -2.9 3.68 1>}.RGB) {transformType=TRANSFORMPOSSOURCE_World}
   Position
   Normal=Transform(None=@n6.RGB)
   TangentU
   TangentV
  ) {materialFunction=/Game/Characters/Player/Male/Medium/Bodies/M_MED_Inferno/FX/Mat/RotateVertexAboutAxis}
  n8 = NamedRerouteUsage()
  n9 = Multiply(A=NamedRerouteUsage(), B=Multiply(A=NamedRerouteUsage(), B=LinearInterpolate(Alpha=@n8) {constA=-166, constB=0}))
  n10 = ScalarParameter() {defaultValue=17, group=PodCatcher_WPO, parameterName=Lids_ClosedHeightOffset}
  n11 = SmoothStep(Value=@n8) {constMax=0.6}
  n12 = NamedRerouteUsage()
  n13 = Constant3Vector() {constant=<rgba 0 0 1 0>}
  n14 = LinearInterpolate(Alpha=NamedRerouteUsage()) {constA=-440, constB=0}
  n15 = OneMinus(None=NamedRerouteUsage())
  n16 = Add(A=Multiply(A=LinearInterpolate(A=@n10, Alpha=@n11) {constB=0}, B=Multiply(A=@n12, B=@n13)), B=Multiply(A=@n13, B=Multiply(A=@n14, B=@n15)))
  Add(A=@n7.WPO, B=Transform(None=Add(A=@n9, B=@n16)) {transformSourceType=TRANSFORMSOURCE_Local})
  n17 = VertexColor()
  n18 = LocalPosition() {includedOffsets=ExcludeOffsets, localOrigin=InstancePreSkinning}
  n19 = ShaderStageSwitch(PixelShader=VertexInterpolator(VS=@n18.XYZ).PS, VertexShader=@n18.XYZ)
  n20 = Saturate(None=ComponentMask(None=@n19) {b=true})
  n21 = NamedRerouteDeclaration(None=Multiply(A=@n17.R, B=OneMinus(None=@n20))) {name=Lids_Mask, nodeColor=<rgba 0.9377 1 0.8953 1>}
  n22 = ComponentMask(None=@n19) {r=true}
  n23 = OneMinus(None=Saturate(None=Add(A=@n22) {constB=-350}))
  n24 = Saturate(None=Add(A=ComponentMask(None=@n19) {g=true}) {constB=-400})
  n25 = Multiply(A=@n21, B=Saturate(None=Add(A=@n23, B=@n24)))
  NamedRerouteDeclaration(None=Add(A=@n21, B=Multiply(A=@n25) {constB=-2})) {name=Lids_Orientation, nodeColor=<rgba 0 1 0.1328 1>}
  n26 = Multiply(A=Subtract(A=@n17.R) {constB=0.6}) {constB=10000}
  NamedRerouteDeclaration(None=Multiply(A=Saturate(None=@n26), B=@n20)) {name=TopCatcher_Mask, nodeColor=<rgba 1 0.0469 0 1>}
  n27 = Multiply(A=@n17.R, B=OneMinus(None=Multiply(A=@n17.R) {constB=1.5}))
  NamedRerouteDeclaration(None=Saturate(None=Multiply(A=@n27) {constB=100000})) {name=BottomCatcher_Mask, nodeColor=<rgba 0 0.2734 1 1>}
  n28 = Time()
  n29 = Add(
   A=ScalarParameter() {group=PodCatcher_Animation, parameterName=RisingUp_Delay, sortPriority=11}
   B=ScalarParameter() {group=PodCatcher_Animation, parameterName=Overall_Delay, sortPriority=11}
  )
  n30 = Multiply(
   A=ScalarParameter() {defaultValue=1, group=PodCatcher_Animation, parameterName=RisingUp_Speed, sortPriority=20}
   B=Subtract(
    A=Subtract(A=@n28, B=@n29) {constB=0}
    B=ScalarParameter() {defaultValue=9999999, group=PodCatcher_Animation, parameterName=BP_GetTime, sortPriority=0}
   )
  )
  NamedRerouteDeclaration(None=SmoothStep(Value=Saturate(None=@n30))) {name=RisingUp_AnimationTime, nodeColor=<rgba 1 0.2812 0 1>}
  n31 = Multiply(A=NamedRerouteUsage(), B=ScalarParameter() {defaultValue=1, group=PodCatcher_Animation, parameterName=Lids_Speed, sortPriority=20})
  n32 = Multiply(A=NamedRerouteUsage(), B=ScalarParameter() {defaultValue=1, group=PodCatcher_Animation, parameterName=TopCatcher_Speed, sortPriority=11})
  n33 = NamedRerouteUsage()
  n34 = ScalarParameter() {defaultValue=1, group=PodCatcher_Animation, parameterName=BottomCatcher_Speed, sortPriority=11}
  n35 = Time()
  n36 = Multiply(A=NamedRerouteUsage(), B=ScalarParameter() {group=PodCatcher_Animation, parameterName=Lids_Delay, sortPriority=20})
  n37 = Multiply(A=NamedRerouteUsage(), B=ScalarParameter() {group=PodCatcher_Animation, parameterName=TopCatcher_Delay, sortPriority=11})
  n38 = NamedRerouteUsage()
  n39 = ScalarParameter() {group=PodCatcher_Animation, parameterName=BottomCatcher_Delay, sortPriority=11}
  n40 = Add(
   A=Add(A=Add(A=@n36, B=@n37), B=Multiply(A=@n38, B=@n39))
   B=ScalarParameter() {group=PodCatcher_Animation, parameterName=Overall_Delay, sortPriority=11}
  )
  n41 = Multiply(
   A=Add(A=Add(A=@n31, B=@n32), B=Multiply(A=@n33, B=@n34))
   B=Subtract(
    A=Subtract(A=@n35, B=@n40) {constB=0}
    B=ScalarParameter() {defaultValue=9999999, group=PodCatcher_Animation, parameterName=BP_GetTime, sortPriority=0}
   )
  ) {constA=1}
  NamedRerouteDeclaration(None=SmoothStep(Value=Saturate(None=@n41))) {name=AnimationTime, nodeColor=<rgba 0.3594 0 1 1>}
  n42 = ScalarParameter() {defaultValue=30, group=PodCatcher_WPO, parameterName=RotationAngle}
  n43 = Normalize(VectorInput=MaterialFunctionCall(R=Cosine(None=@n42) {period=360}, G=Sine(None=@n42) {period=360}, B=Constant()) {materialFunction=&AppendMany}.RGB)
  NamedRerouteDeclaration(None=@n43) {name=RotationAxisVector, nodeColor=<rgba 0 1 0.0156 1>}
  Reroute(None=@n43)
  TextureSampleParameter2D() {
   group=Basic_PBR
   parameterName=BaseColor_Texture
   texture=/BarnTowelArmoryGameplay/Items/Environmental/ArmoryLanding/Textures/T_MobileArmory_Launch_Pad_Ramp_D
  }
  TextureSampleParameter2D() {
   group=Basic_PBR
   parameterName=SRM_Texture
   samplerType=SAMPLERTYPE_Masks
   texture=/BarnTowelArmoryGameplay/Items/Environmental/ArmoryLanding/Textures/T_MobileArmory_Launch_Pad_Ramp_S
  }
  Transform(None=@n7.Normal) {transformSourceType=TRANSFORMSOURCE_World, transformType=TRANSFORM_Tangent}
 ]
}
